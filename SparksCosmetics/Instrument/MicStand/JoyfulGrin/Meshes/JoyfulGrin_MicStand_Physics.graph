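# Physics asset for the JoyfulGrin mic stand instrument cosmetic in Sparks, simulating a dangling chain on the stand.
# Capsule bodies cover the long handle, two chain roots, thirteen links and a middle connector.
# Links 01–06 run from the top root, which is constrained to the handle, and 13–07 run from the bottom root, which has no handle constraint.
# The halves meet at the connector through angularly locked joints, while most other joints allow limited swing and twist.
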PhysicsAsset /SparksCosmetics/Instrument/MicStand/JoyfulGrin/Meshes/JoyfulGrin_MicStand_Physics
bodies {
 chain_bottom_root [{shapeType=Capsule, length=0, radius=1.5534, shapeName=chain_bottom_root_capsule, center=<0 -0.4467 0>}]
 chain_top_root [{shapeType=Capsule, length=1, radius=1, center=<0 0 -0.775>}]
 dyn_link_01 [{shapeType=Capsule, length=3.0449, radius=1, center=<2.08 0 0>, rotation=<rot 90 0 0>}]
 dyn_link_02 [{shapeType=Capsule, length=3.0449, radius=1, center=<2.08 0 0>, rotation=<rot 90 0 0>}]
 dyn_link_03 [{shapeType=Capsule, length=3.0449, radius=1, center=<2.08 0 0>, rotation=<rot 90 0 0>}]
 dyn_link_04 [{shapeType=Capsule, length=3.0449, radius=1, center=<2.08 0 0>, rotation=<rot 90 0 0>}]
 dyn_link_05 [{shapeType=Capsule, length=3.0449, radius=1, center=<2.08 0 0>, rotation=<rot 90 0 0>}]
 dyn_link_06 [{shapeType=Capsule, length=2.5449, radius=1.5, center=<2.08 0 0>, rotation=<rot 90 0 0>}]
 dyn_link_07 [{shapeType=Capsule, length=2.3664, radius=1.6135, shapeName=dyn_link_07_capsule, center=<-1.7726 0 0.0402>, rotation=<rot 90 0 0>}]
 dyn_link_08 [{
   shapeType=Capsule
   length=2.3657
   radius=1.6121
   shapeName=dyn_link_08_capsule
   center=<-1.9356 0.0004 -0.0272>
   rotation=<rot 0 -90.3213 -87.718>
  }]
 dyn_link_09 [{
   shapeType=Capsule
   length=2.3633
   radius=1.6076
   shapeName=dyn_link_09_capsule
   center=<-1.8732 -0.028 0.0025>
   rotation=<rot 0 -89.78 -87.7404>
  }]
 dyn_link_10 [{
   shapeType=Capsule
   length=2.3657
   radius=1.6121
   shapeName=dyn_link_10_capsule
   center=<-1.9043 0.036 0.084>
   rotation=<rot 0 -90.2732 -87.6679>
  }]
 dyn_link_11 [{
   shapeType=Capsule
   length=2.3657
   radius=1.6121
   shapeName=dyn_link_11_capsule
   center=<-1.8572 -0.0203 0.0255>
   rotation=<rot 0 89.9057 -86.2571>
  }]
 dyn_link_12 [{
   shapeType=Capsule
   length=2.3657
   radius=1.6121
   shapeName=dyn_link_12_capsule
   center=<-1.8006 0.0439 0.0174>
   rotation=<rot 0 -90.153 -87.384>
  }]
 dyn_link_13 [{
   shapeType=Capsule
   length=1.7994
   radius=1.6516
   shapeName=dyn_link_13_capsule
   center=<-1.6522 -0.047 -0.1455>
   rotation=<rot 0 -90.2639 -85.774>
  }]
 dyn_link_connector [{shapeType=Capsule, length=2.6798, radius=1.6, center=<1.4654 0 -0.0379>, rotation=<rot 60.0001 89.9995 89.9995>}]
 handle [{shapeType=Capsule, length=98.3905, radius=1.9997, shapeName=handle_capsule, center=<0 0 19.9625>, rotation=<rot 0 -90 0>}]
}
constraints [
 {bone1Name=dyn_link_01, bone2Name=chain_top_root, swing1LimitDegrees=45, swing2LimitDegrees=45, twistLimitDegrees=15}
 {bone1Name=dyn_link_02, bone2Name=dyn_link_01, swing1LimitDegrees=45, swing2LimitDegrees=45, twistLimitDegrees=15}
 {bone1Name=dyn_link_03, bone2Name=dyn_link_02, swing1LimitDegrees=45, swing2LimitDegrees=45, twistLimitDegrees=15}
 {bone1Name=dyn_link_04, bone2Name=dyn_link_03, swing1LimitDegrees=45, swing2LimitDegrees=45, twistLimitDegrees=15}
 {bone1Name=dyn_link_05, bone2Name=dyn_link_04, swing1LimitDegrees=45, swing2LimitDegrees=45, twistLimitDegrees=15}
 {bone1Name=dyn_link_06, bone2Name=dyn_link_05, swing1LimitDegrees=45, swing2LimitDegrees=45, twistLimitDegrees=15}
 {bone1Name=chain_top_root, bone2Name=handle, swing1LimitDegrees=45, swing2LimitDegrees=45, twistLimitDegrees=15}
 {
  bone1Name=dyn_link_connector
  bone2Name=dyn_link_06
  swing1LimitDegrees=15
  swing1Motion=Locked
  swing2LimitDegrees=15
  swing2Motion=Locked
  twistLimitDegrees=15
  twistMotion=Locked
 }
 {bone1Name=dyn_link_13, bone2Name=chain_bottom_root, swing1LimitDegrees=90, swing2LimitDegrees=45, twistLimitDegrees=90}
 {bone1Name=dyn_link_12, bone2Name=dyn_link_13, swing1LimitDegrees=45, swing2LimitDegrees=45, twistLimitDegrees=15}
 {bone1Name=dyn_link_11, bone2Name=dyn_link_12, swing1LimitDegrees=45, swing2LimitDegrees=45, twistLimitDegrees=15}
 {bone1Name=dyn_link_10, bone2Name=dyn_link_11, swing1LimitDegrees=15, swing2LimitDegrees=15, twistLimitDegrees=5}
 {bone1Name=dyn_link_09, bone2Name=dyn_link_10, swing1LimitDegrees=45, swing2LimitDegrees=45, twistLimitDegrees=15}
 {bone1Name=dyn_link_08, bone2Name=dyn_link_09, swing1LimitDegrees=45, swing2LimitDegrees=45, twistLimitDegrees=15}
 {
  bone1Name=dyn_link_07
  bone2Name=dyn_link_08
  swing1LimitDegrees=15
  swing1Motion=Locked
  swing2LimitDegrees=15
  swing2Motion=Locked
  twistLimitDegrees=10
  twistMotion=Locked
 }
 {
  bone1Name=dyn_link_07
  bone2Name=dyn_link_connector
  swing1LimitDegrees=15
  swing1Motion=Locked
  swing2LimitDegrees=15
  swing2Motion=Locked
  twistLimitDegrees=15
  twistMotion=Locked
 }
]
